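# Additive unlit Niagara ribbon material for the rainbow trail on the GlitterMane cosmetic shoes.
# A scaled, quantized V gradient drives chained lerps through four color parameters and back to the first, forming stripes tinted by particle color and passed through emissive adjustments.
# A biased sine along U fades both ends, an edge mask softens the sides, and particle alpha scales both.
# Depth fade applies except on low and medium quality.
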Material /CosmeticShoes/Assets/FORT_Shoes/Shoe_GlitterMane/FX/M_Shoe_GlitterMane_RainbowTrail
graph {
 properties {
  bAutomaticallySetUsageInEditor=false
  bUsedWithNiagaraRibbons=true
  blendMode=BLEND_Additive
  shadingModel=MSM_Unlit
  usedShadingModels=Unlit
 }
 nodes [
  n1 = ParticleColor()
  n2 = VectorParameter() {parameterName=Color A, defaultValue=<rgba 0.95 0.8027 0.0665 0.3>}
  n3 = VectorParameter() {parameterName=Color B, defaultValue=<rgba 1 0.2633 0.15 0.2>}
  n4 = ComponentMask(None=TextureCoordinate()) {g=true}
  n5 = Add(A=Multiply(A=Subtract(A=@n4) {constB=0.5}, B=ScalarParameter() {defaultValue=0.95, parameterName=Gradient Scale})) {constB=0.5}
  n6 = Constant() {r=5}
  n7 = Multiply(
   A=MaterialFunctionCall!(Gradient=Saturate(None=@n5), Steps=@n6) {materialFunction=/Game/Effects/Fort_Effects/MaterialFunctions/MF_Quantize}.Increased Range
   B=Subtract(A=@n6)
  ) {constB=4}
  n8 = Saturate(None=@n7) {desc=Color 1 to 2}
  n9 = Subtract(A=@n7)
  n10 = Subtract(A=@n9)
  n11 = LinearInterpolate(
   A=LinearInterpolate(
    A=LinearInterpolate(A=@n2.RGB, B=@n3.RGB, Alpha=@n8)
    B=VectorParameter() {parameterName=Color C, defaultValue=<rgba 0.9 0.09 0.9 0.2>}.RGB
    Alpha=Saturate(None=@n9) {desc=1+2 to 3}
   )
   B=VectorParameter() {parameterName=Color D, defaultValue=<rgba 0.14 0.5224 1 0.1>}.RGB
   Alpha=Saturate(None=@n10) {desc=1+2+3 to 4}
  )
  n12 = Saturate(None=Subtract(A=@n10)) {desc=1+2+3+4 to 5}
  MaterialFunctionCall(Emissive=Multiply(A=@n1.RGB, B=LinearInterpolate(A=@n11, B=@n2.RGB, Alpha=@n12))) {materialFunction=&MF_FX_EmissiveAdjustments}
  n13 = TextureCoordinate()
  n14 = ComponentMask(None=@n13) {r=true}
  n15 = Divide(A=Power(Base=OneMinus(None=@n14), Exp=ScalarParameter() {defaultValue=2, parameterName=Fade Bias Exponent}) {constExponent=3})
  n16 = ConstantBiasScale(None=ComponentMask(None=@n13) {g=true}) {bias=-0.5, scale=2}
  n17 = Divide(A=OneMinus(None=Abs(None=@n16)), B=ScalarParameter() {defaultValue=0.1, parameterName=Fade Sides Size})
  n18 = Multiply(A=@n1.A, B=Multiply(A=Sine(None=@n15), B=Saturate(None=@n17)))
  QualitySwitch(Default=DepthFade(Opacity=@n18, FadeDistance=ScalarParameter() {defaultValue=5, parameterName=Depth Fade Distance}), Low=@n18, Medium=@n18)
 ]
}
