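# Niagara burst effect for the GlitterMane cosmetic shoe's landing dust, pairing multicolor smoke with gold glitter.
# SmokePuff bursts sprites outward from spheres, stepping through four colors by execution index plus engine time, then expands and fades them under drag and upward force.
# NE_GLints bursts short-lived gold sprites whose randomized looping GlitterAlpha curve drives twinkling color boosts and flare size and color attributes.
# Both emitters use two sprite renderers and have source-movement inheritance disabled.
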NiagaraSystem /CosmeticShoes/Assets/FORT_Shoes/Shoe_GlitterMane/FX/NS_Shoe_GlitterMane_DustLanding
dependencies {
 usedDataInterfaces [&NiagaraDataInterfaceColorCurve, &NiagaraDataInterfaceCurve]
 usedDynamicInputs []
 usedModules [&SystemState]
}
emitters {
 NE_GLints {
  emitterUpdateScript [
   &EmitterState()
   &SpawnBurst_Instantaneous(Spawn Count=10, Use Spawn Probability=false, UseLoopCountLimit=false)
  ]
  particleSpawnScript [
   &InitializeParticle(
    Color=&RandomRangeLinearColor(Minimum=<rgba 1 0.6889 0 1>, Maximum=<rgba 1 0.3404 0 1>)
    Lifetime Max=0.6
    Lifetime Min=0.5
    Lifetime Mode=::NewEnumerator1
    Sprite Rotation Mode=::NewEnumerator2
    Sprite Size Mode=::NewEnumerator3
    Uniform Sprite Size=&RandomRangeFloat(Minimum=8, Maximum=10)
    UsePositionOffset=false
    Write Material Random=true
   )
   &ShapeLocation(
    MaskSpawnGroup=false
    Sphere Distribution=::NewEnumerator1
    Sphere Radius=&RandomRangeFloat(Minimum=40, Maximum=20)
    U Position=&ReturnNormalizedExecIndex()
    UseOwnerScale=true
    V Position=&RandomRangeFloat(Minimum=0.5, Maximum=0.7)
   )
   &AddVelocity(
    Rotation Coordinate Space=::Simulation
    Velocity=&Multiply_VectorByFloat(Vector=$Particles.ShapeLocation.ShapeVector, Float=0.02)
    Velocity Speed Scale=&RandomRangeFloat(Minimum=200, Maximum=300)
   )
  ]
  particleUpdateScript [
   /SetVariables(Particles.GlitterAlpha:NiagaraFloat=&FloatFromCurve(
     FloatCurve=DataInterface() {propertyValues {Curve {keys [
         {interpMode=RCIM_Cubic, tangentMode=RCTM_User, time=0, value=0}
         {interpMode=RCIM_Cubic, tangentMode=RCTM_User, time=0.05, value=1}
         {interpMode=RCIM_Cubic, tangentMode=RCTM_User, time=0.25, value=0}
        ]}}}
     CurveIndex=&ModuloFloat(
      A=&Add_Float(
       A=&Multiply_Float(A=$Emitter.Age, B=2)
       B=&RandomRangeFloat(Minimum=0, Maximum=1)
      )
      B=1
     )
    )) {isSetParameters=true}
   &ParticleState()
   &ScaleSpriteSize(Uniform Curve Sprite Scale=DataInterface() {propertyValues {Curve {keys [
        {arriveTangent=19.4641, interpMode=RCIM_Cubic, leaveTangent=19.4641, tangentMode=RCTM_User, time=0, value=0}
        {interpMode=RCIM_Cubic, tangentMode=RCTM_User, time=0.05, value=1}
        {arriveTangent=-3.5627, interpMode=RCIM_Cubic, leaveTangent=-3.5627, tangentMode=RCTM_User, time=1, value=0}
       ]}}})
   &Drag(Drag=4, Use Linear Drag=true, Use Rotational Drag=false)
   &InheritSourceMovement(
    Applied Position Scale=&VectorFromFloat(Value=-1)
    ApplyForce=false
    ApplyPosition=true
    ApplyPositionFalloff=false
   ) {enabled=false}
   &SpriteRotationRate(Rotation Rate=&RandomRangeFloat(Minimum=-20, Maximum=20))
   &ScaleColor(
    Scale RGB=&MakeVectorFromLinearColorRGB(Color=&LerpLinearColors(StartColor=<rgba 1 1 1 1>, EndColor=<rgba 10 6.8887 0 1>, LerpFactor=$Particles.GlitterAlpha))
    ScaleA=true
    ScaleRGB=true
   )
   /SetVariables(Particles.FLareSize:Vector2f=&Multiply_Vector2D_ByFloat(
     Float=&Lerp_Float(A=0, B=1, Alpha=$Particles.GlitterAlpha)
     Vector2D=&Multiply_Vector2D(A=$Particles.SpriteSize, B=<1 1>)
    )) {isSetParameters=true}
   /SetVariables(Particles.FlareColor=&ColorFromCurve(
     ColorCurve=DataInterface() {propertyValues {
       AlphaCurve {keys [<at 0 0 RCIM_Cubic>,<at 1 0.2 RCIM_Cubic>]}
       BlueCurve {keys [<at 0 0 RCIM_Cubic>,<at 1 0 RCIM_Cubic>]}
       GreenCurve {keys [<at 0 0.383 RCIM_Cubic>,<at 1 0.4596 RCIM_Cubic>]}
       RedCurve {keys [<at 0 1 RCIM_Cubic>,<at 1 1 RCIM_Cubic>]}
      }}
     CurveIndex=$Particles.GlitterAlpha
    )) {isSetParameters=true}
   &SolveForcesAndVelocity(Clamp Velocity=false, Limit Acceleration=false)
   &DynamicMaterialParameters(
    Index 0 Param 1=1
    Index 0 Param 2=0.5
    Module.Index0Param0Write=true
    Module.Index0Param1Write=true
    Module.Index0Param2Write=true
    Module.Index0Param3Write=true
   )
  ]
  rendererClasses [&NiagaraSpriteRendererProperties]
  renderers [
   {rendererClass=&NiagaraSpriteRendererProperties}
   {rendererClass=&NiagaraSpriteRendererProperties, rendererIndex=1}
  ]
 }
 SmokePuff {
  emitterUpdateScript [
   &EmitterState()
   &SpawnBurst_Instantaneous(
    Spawn Count=&RandomRangeInt(Minimum=16, Maximum=20)
    Use Spawn Probability=false
    UseLoopCountLimit=false
   )
  ]
  particleSpawnScript [
   &InitializeParticle(
    Color=&ColorFromCurve(
     ColorCurve=DataInterface() {propertyValues {
       AlphaCurve {keys [<at 0 1 RCIM_Constant>]}
       BlueCurve {keys [<at 0 0.15 RCIM_Constant>,<at 0.25 0.3 RCIM_Constant>,<at 0.5 1.5 RCIM_Constant>,<at 0.75 1.5 RCIM_Constant>]}
       GreenCurve {keys [<at 0 1.275 RCIM_Constant>,<at 0.25 0.46 RCIM_Constant>,<at 0.5 0.45 RCIM_Constant>,<at 0.75 0.84 RCIM_Constant>]}
       RedCurve {keys [<at 0 1.5 RCIM_Constant>,<at 0.25 1.5 RCIM_Constant>,<at 0.5 1.5 RCIM_Constant>,<at 0.75 0.3 RCIM_Constant>]}
      }}
     CurveIndex=&ModuloFloat(
      A=&Add_Float(A=&ReturnNormalizedExecIndex(), B=&Frac_Float(Float=$Engine.Time))
      B=1
     )
    )
    Lifetime Max=0.7
    Lifetime Min=0.6
    Lifetime Mode=::NewEnumerator1
    Mesh Scale Mode=::NewEnumerator3
    Mesh Uniform Scale=150
    Sprite Rotation Mode=::NewEnumerator2
    Sprite Size Mode=::NewEnumerator1
    Uniform Sprite Size Max=5
    Uniform Sprite Size Min=4
    UsePositionOffset=false
    Write Material Random=true
   ) {module=InitializeParticle002}
   &ShapeLocation(
    MaskSpawnGroup=false
    Sphere Distribution=::NewEnumerator1
    Sphere Radius=&RandomRangeFloat(Minimum=10, Maximum=20)
    U Position=&ReturnNormalizedExecIndex()
    UseOwnerScale=true
    V Position=&RandomRangeFloat(Minimum=0.5, Maximum=0.55)
   )
   &AddVelocity(
    Rotation Coordinate Space=::Simulation
    Velocity=$Particles.ShapeLocation.ShapeNormal
    Velocity Speed Scale=&RandomRangeFloat(Minimum=250, Maximum=350)
   )
  ]
  particleUpdateScript [
   &ParticleState()
   &GravityForce(Gravity=<0 0 200>)
   &Drag(Drag=5, Use Linear Drag=true, Use Rotational Drag=false)
   &InheritSourceMovement(
    Applied Position Scale=&VectorFromFloat(Value=-1)
    ApplyForce=false
    ApplyPosition=true
    ApplyPositionFalloff=false
   ) {enabled=false}
   &ScaleColor(
    Scale Alpha=&FloatFromCurve(
     FloatCurve=DataInterface() {propertyValues {Curve {keys [
         {arriveTangent=7.3326, interpMode=RCIM_Cubic, leaveTangent=7.3326, tangentMode=RCTM_User, time=0, value=0}
         {arriveTangent=0.0071, interpMode=RCIM_Cubic, leaveTangent=0.0071, tangentMode=RCTM_User, time=0.4, value=1}
         <at 1 0 RCIM_Cubic>
        ]}}}
     Scale Curve=0.4
    )
    Scale RGB=&MakeVectorFromLinearColorRGB(Color=&ColorFromCurve(ColorCurve=DataInterface() {propertyValues {
        AlphaCurve {keys [<at 0 1 RCIM_Cubic>,<at 1 0 RCIM_Cubic>]}
        BlueCurve {keys [<at 0 1.2 RCIM_Cubic>,<at 0.4269 1 RCIM_Cubic>]}
        GreenCurve {keys [<at 0 1.2 RCIM_Cubic>,<at 0.4269 1 RCIM_Cubic>]}
        RedCurve {keys [<at 0 1.2 RCIM_Cubic>,<at 0.4269 1 RCIM_Cubic>]}
       }}))
    ScaleA=true
    ScaleRGB=true
   ) {module=ScaleColor001}
   &DynamicMaterialParameters(
    Index 0 Param 1=&Add_Float(
     A=&FloatFromCurve(FloatCurve=DataInterface() {propertyValues {Curve {keys [{arriveTangent=2.1549, time=0, value=0.8}, <at 1 1>]}}})
     B=&Multiply_Float(A=$Particles.Initial.MaterialRandom, B=0.5)
    )
    Index 1 Param 1=1
    Index 1 Param 2=1
    Index 1 Param 3=1
    Index 1 Param 4=1
    Index 2 Param 1=&FloatFromCurve(FloatCurve=DataInterface() {propertyValues {Curve {keys [
         {arriveTangent=1.8288, interpMode=RCIM_Cubic, leaveTangent=1.8288, tangentMode=RCTM_User, time=0, value=0.3}
         {arriveTangent=-0.0004, interpMode=RCIM_Cubic, leaveTangent=-0.0004, tangentMode=RCTM_User, time=1, value=1}
        ]}}})
    Index 2 Param 2=&RandomRangeFloat(Minimum=0.5, Maximum=1)
    Index 2 Param 3=4
    Index 2 Param 4=&RandomRangeFloat(Minimum=0.8, Maximum=1.2)
    Module.Param0WriteEnabled=true
    Module.Param0WriteEnabled1=true
    Module.Param0WriteEnabled2=true
    Module.Param0WriteEnabled3=true
    Module.Param1WriteEnabled=true
    Module.Param1WriteEnabled1=true
    Module.Param1WriteEnabled2=true
    Module.Param1WriteEnabled3=true
    Module.Param2WriteEnabled=true
    Module.Param2WriteEnabled1=true
    Module.Param2WriteEnabled2=true
    Module.Param2WriteEnabled3=true
    Module.Param3WriteEnabled=true
    Module.Param3WriteEnabled1=true
    Module.Param3WriteEnabled2=true
    Module.Param3WriteEnabled3=true
    Write Parameter Index 1=true
    Write Parameter Index 2=true
    Write Parameter Index 3=true
   ) {module=DynamicMaterialParameters002}
   &ScaleSpriteSize(Scale Factor=&Vector2DFromFloat(Value=&FloatFromCurve(FloatCurve=DataInterface() {propertyValues {Curve {keys [
          {arriveTangent=87.4549, interpMode=RCIM_Cubic, leaveTangent=87.4549, tangentMode=RCTM_User, time=0, value=1}
          {arriveTangent=4.8897, interpMode=RCIM_Cubic, leaveTangent=4.8897, tangentMode=RCTM_User, time=0.1, value=8}
          {interpMode=RCIM_Cubic, leaveTangent=-6.7553, time=1.0011, value=10}
         ]}}})))
   &CameraOffset(Camera Offset Amount=50)
   &SpriteRotationRate(Rotation Rate=&RandomRangeFloat(Minimum=-90, Maximum=90))
   &SpriteRotationRate(Rotation Rate=&Multiply_Float(A=90, B=&RandomRangeFloat(Minimum=-1, Maximum=1))) {module=SpriteRotationRate001}
   &SolveForcesAndVelocity(Clamp Velocity=false, Limit Acceleration=false)
  ]
  rendererClasses [&NiagaraSpriteRendererProperties]
  renderers [
   {rendererClass=&NiagaraSpriteRendererProperties}
   {rendererClass=&NiagaraSpriteRendererProperties, rendererIndex=1}
  ]
 }
}
summary {
 systemName=NS_Shoe_GlitterMane_DustLanding
 emitters [
  {emitterName=SmokePuff, rendererClasses [&NiagaraSpriteRendererProperties]}
  {emitterName=NE_GLints, rendererClasses [&NiagaraSpriteRendererProperties]}
 ]
 userVariables [User.IsLocalPlayer {type=NiagaraBool, defaultValue=true}]
}
