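# Niagara system for the GlitterMane shoe cosmetic that draws a rainbow ribbon trail at each foot, with sparkling glints along it.
# NE_RainbowTrail burst-spawns two partitioned ribbons, placing points procedurally from skeletal mesh bones and stretching the tails by a smoothed motion-lag vector.
# NE_Glints reads those ribbons to position pulsing, flaring sprites, with spawning gated by System.ShowTrails.
# Per-shoe scale affects ribbon and glint size; time of day tints or dims both.
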NiagaraSystem /CosmeticShoes/Assets/FORT_Shoes/Shoe_GlitterMane/FX/NS_Shoe_GlitterMane_RainbowTrails
dependencies {
 usedDataInterfaces [
  &NiagaraDataInterfaceArrayFloat
  &NiagaraDataInterfaceCurve
  &NiagaraDataInterfaceParticleRead
  &NiagaraDataInterfaceColorCurve
  &NiagaraDataInterfaceSkeletalMesh
 ]
 usedDynamicInputs [&InvertBool, &RemapRange]
 usedModules [
  &self:SystemSpawnScript.SystemScriptSource.SystemScriptGraph.NiagaraNodeAssignment_0.SetVariables_#1
  &self:SystemSpawnScript.SystemScriptSource.SystemScriptGraph.NiagaraNodeAssignment_3.SetVariables_#2
  &self:SystemSpawnScript.SystemScriptSource.SystemScriptGraph.NiagaraNodeAssignment_5.SetVariables_#3
  &SystemState
  &self:SystemSpawnScript.SystemScriptSource.SystemScriptGraph.NiagaraNodeAssignment_2.SetVariables_#4
  &self:SystemSpawnScript.SystemScriptSource.SystemScriptGraph.NiagaraNodeAssignment_4.SetVariables_#5
 ]
}
emitters {
 NE_Glints {
  emitterUpdateScript [
   &EmitterState()
   &SpawnRate(
    Spawn Probability=&MakeCustomFloatFromBool(Bool=$System.ShowTrails, True Float=1, False Float=0)
    SpawnRate=5
    Use Spawn Probability=true
   )
  ]
  particleSpawnScript [
   &InitializeParticle(
    Color=&RandomRangeLinearColor(Minimum=<rgba 1 0.6889 0 1>, Maximum=<rgba 1 0.3404 0 1>)
    Lifetime Max=1
    Lifetime Min=0.5
    Lifetime Mode=::NewEnumerator1
    Sprite Rotation Mode=::NewEnumerator2
    Sprite Size Mode=::NewEnumerator3
    Uniform Sprite Size=&RandomRangeFloat(Minimum=8, Maximum=10)
    UsePositionOffset=false
    Write Material Random=true
   )
   &PartitionParticles()
  ]
  particleUpdateScript [
   /SetVariables(Particles.GlitterAlpha:NiagaraFloat=&FloatFromCurve(
     FloatCurve=DataInterface() {propertyValues {Curve {keys [
         {interpMode=RCIM_Cubic, tangentMode=RCTM_User, time=0, value=0}
         {interpMode=RCIM_Cubic, tangentMode=RCTM_User, time=0.05, value=1}
         {interpMode=RCIM_Cubic, tangentMode=RCTM_User, time=0.25, value=0}
        ]}}}
     CurveIndex=&ModuloFloat(
      A=&Add_Float(
       A=&Multiply_Float(A=$Emitter.Age, B=2)
       B=&RandomRangeFloat(Minimum=0, Maximum=1)
      )
      B=1
     )
    )) {isSetParameters=true}
   &self:FollowRibbon(
    Alpha:NiagaraFloat=&FloatFromCurve(FloatCurve=DataInterface() {propertyValues {Curve {keys [<at 0 0>,<at 1 1>]}}})
    ParticlePerRibbon:NiagaraInt32=$System.ParticlePerRibbon
    RibbonNumber:NiagaraInt32=$Particles.PartitionParticles.Partition
    RibbonReader=DataInterface() {propertyValues {EmitterBinding {bindingMode=Other, emitterName=NE_RainBowTrail}}}
   ) {module=FollowRibbon}
   /SetVariables(Particles.Position=&AddVectorToPosition(Position=$Output.FollowRibbon.RibbonPos, Vector=&RandomVector(Vector Scale=3))) {isSetParameters=true}
   &ParticleState()
   &ScaleSpriteSize(
    Uniform Curve Scale=&MergeHelper_FloatSwitch(
     0=&MakeFloatFromVector(Vector=$User.ShoeScale_L, Channel=::NewEnumerator0)
     1=&MakeFloatFromVector(Vector=$User.ShoeScale_R, Channel=::NewEnumerator0)
     2=?
     3=?
     Index=$Particles.PartitionParticles.Partition
     Num=::NewEnumerator0
    )
    Uniform Curve Sprite Scale=DataInterface() {propertyValues {Curve {keys [
        {arriveTangent=19.4641, interpMode=RCIM_Cubic, leaveTangent=19.4641, tangentMode=RCTM_User, time=0, value=0}
        {interpMode=RCIM_Cubic, tangentMode=RCTM_User, time=0.05, value=1}
        {arriveTangent=-3.5627, interpMode=RCIM_Cubic, leaveTangent=-3.5627, tangentMode=RCTM_User, time=1, value=0}
       ]}}}
   )
   &SpriteRotationRate(Rotation Rate=&RandomRangeFloat(Minimum=-20, Maximum=20))
   &ScaleColor(
    Scale RGB=&MakeVectorFromLinearColorRGB(Color=&Multiply_LinearColorByFloat(
      Linear Color=&LerpLinearColors(StartColor=<rgba 1 1 1 1>, EndColor=<rgba 10 6.8887 0 1>, LerpFactor=$Particles.GlitterAlpha)
      Float=&FNTimeOfDayScalar(Night Scale=0.3, Morning Scale=0.3, Evening Scale=0.3)
     ))
    ScaleA=true
    ScaleRGB=true
   )
   /SetVariables(Particles.FLareSize:Vector2f=&Multiply_Vector2D_ByFloat(
     Float=&Lerp_Float(A=0, B=1, Alpha=$Particles.GlitterAlpha)
     Vector2D=&Multiply_Vector2D(A=$Particles.SpriteSize, B=<1 1>)
    )) {isSetParameters=true}
   /SetVariables(Particles.FlareColor=&ColorFromCurve(
     ColorCurve=DataInterface() {propertyValues {
       AlphaCurve {keys [<at 0 0 RCIM_Cubic>,<at 1 0.2 RCIM_Cubic>]}
       BlueCurve {keys [<at 0 0 RCIM_Cubic>,<at 1 0 RCIM_Cubic>]}
       GreenCurve {keys [<at 0 0.383 RCIM_Cubic>,<at 1 0.4596 RCIM_Cubic>]}
       RedCurve {keys [<at 0 1 RCIM_Cubic>,<at 1 1 RCIM_Cubic>]}
      }}
     CurveIndex=$Particles.GlitterAlpha
    )) {isSetParameters=true}
   &SolveForcesAndVelocity(Clamp Velocity=false, Limit Acceleration=false)
   &DynamicMaterialParameters(
    Index 0 Param 1=1
    Index 0 Param 2=0.5
    Module.Index0Param0Write=true
    Module.Index0Param1Write=true
    Module.Index0Param2Write=true
    Module.Index0Param3Write=true
   )
   &KillParticles(Kill Particles=&SetBoolByFloatComparison(A=&MakeFloatFromVector2d(Vector 2D=$Particles.SpriteSize), B=0, Comparison Type=::NewEnumerator4))
  ]
  rendererClasses [&NiagaraSpriteRendererProperties]
  renderers [
   {rendererClass=&NiagaraSpriteRendererProperties}
   {rendererClass=&NiagaraSpriteRendererProperties, rendererIndex=1}
  ]
 }
 NE_RainbowTrail {
  emitterUpdateScript [
   &EmitterState(
    Inactive Response=::NewEnumerator1
    Life Cycle Mode=::NewEnumerator1
    Loop Behavior=::NewEnumerator1
    Loop Duration Mode=::NewEnumerator1
    UseLoopDelay=false
   )
   &SpawnBurst_Instantaneous(
    Spawn Count=&Multiply_Int(A=2, B=$System.ParticlePerRibbon)
    Use Spawn Probability=false
    UseLoopCountLimit=false
   )
  ]
  particleSpawnScript [
   &PartitionParticles(
    Normalized Index In Partition Includes One=true
    Particles Per Partition:NiagaraInt32=$System.ParticlePerRibbon
    Partition Mode:ENiagara_PartitionMode_v2=::NewEnumerator1
    Write Partition Index to Ribbon Link Order:ENiagara_AttributeSamplingApplyOutput=::NewEnumerator0
    Write Partition to Ribbon ID=::NewEnumerator0
   )
   &InitializeParticle(
    Color Mode=::NewEnumerator0
    Lifetime=5
    Ribbon Width=&Multiply_Float(
     A=&Lerp_Float(A=$System.TrailWidthMedium, B=$System.TrailWidthLarge, Alpha=$System.ScaleAlpha)
     B=&MergeHelper_FloatSwitch(
      0=&MakeFloatFromVector(Vector=$User.ShoeScale_R, Channel=::NewEnumerator0)
      1=&MakeFloatFromVector(Vector=$User.ShoeScale_L, Channel=::NewEnumerator0)
      2=?
      3=?
      Index=$Particles.PartitionParticles.Partition
      Num=::NewEnumerator0
     )
    )
    Ribbon Width Mode=::NewEnumerator1
    UsePositionOffset=false
    Write Material Random=true
   )
   /SetVariables(Particles.RibbonU0Override=$Particles.PartitionParticles.Normalized Index In Partition) {isSetParameters=true}
   &KillParticles(Kill Particles=&SetBoolByFloatComparison(A=$Particles.RibbonWidth, B=0, Comparison Type=::NewEnumerator4))
  ]
  particleUpdateScript [
   &ParticleState(Kill Particles When Lifetime Has Elapsed=false)
   &SkeletalMeshLocation(
    Bone / Socket Index=$Particles.PartitionParticles.Partition
    Bone Sampling Mode=::NewEnumerator4
    MaskSpawnGroup=false
    Orientation Sampling=::NewEnumerator1
    Position Sampling=::NewEnumerator2
    Sampled Position Offset=&Multiply_VectorByFloat(
     Vector=&Lerp_Vector(A=<3 4 0.5>, B=<0 12 1>, Alpha=$System.ScaleAlpha)
     Float=&SelectFloatFromArray(Direct Array Index=$Particles.PartitionParticles.Partition, Float Selection Array=DataInterface() {propertyValues {FloatData [1,-1]}})
    )
    Skeletal Mesh=$System.Skeletal Mesh
   )
   &InterpolateOverTime(Rate Of Change=9, Target Position Value:NiagaraPosition=$StackContext.SkeletalMeshLocation.SampledPosition, Variable Type=::NewEnumerator6)
   /SetVariables(Particles.TrailStretchVector:Vector3f=&Multiply_Vector(
     A=&Subtract_Position(A=$StackContext.InterpolateOverTime.InterpolateOverTime_Position.Moving Average, B=$StackContext.SkeletalMeshLocation.SampledPosition)
     B=<1 1 0.666>
    )) {isSetParameters=true}
   /SetVariables(Particles.Position=&AddVectorToPosition(
     Position=$Particles.SkeletalMeshLocation.SampledPosition
     Vector=&Add_Vector(
      A=&RotateVectorByQuaternion(
       Vector To Rotate=&Multiply_VectorByFloat(
        Vector=&Add_Vector(
         A=&Lerp_Vector(
          A=<0 0 0>
          B=&Multiply_VectorByFloat(
           Vector=<0 1 0>
           Float=&Lerp_Float(A=$System.TrailLengthMedium, B=$System.TrailLengthLarge, Alpha=$System.ScaleAlpha)
          )
          Alpha=$Particles.PartitionParticles.Normalized Index In Partition
         )
         B=&Lerp_Vector(
          A=<0 0 0>
          B=<-11 0 0>
          Alpha=&FloatFromCurve(
           FloatCurve=DataInterface() {propertyValues {Curve {keys [
               {interpMode=RCIM_Cubic, tangentMode=RCTM_User, time=0, value=0}
               {arriveTangent=2.6319, interpMode=RCIM_Cubic, leaveTangent=2.6319, tangentMode=RCTM_User, time=1, value=1}
              ]}}}
           CurveIndex=$Particles.PartitionParticles.Normalized Index In Partition
          )
         )
        )
        Float=&Lerp_Float(A=1, B=-1, Alpha=&MakeFloatFromInt(Int=$Particles.PartitionParticles.Partition))
       )
       Quaternion=$StackContext.SkeletalMeshLocation.SampledRotation
      )
      B=&Multiply_VectorByFloat(
       Vector=$StackContext.TrailStretchVector
       Float=&FloatFromCurve(
        FloatCurve=DataInterface() {propertyValues {Curve {keys [
            {interpMode=RCIM_Cubic, tangentMode=RCTM_User, time=0, value=0}
            {arriveTangent=2.6319, interpMode=RCIM_Cubic, leaveTangent=2.6319, tangentMode=RCTM_User, time=1, value=1}
           ]}}}
        CurveIndex=$Particles.Initial.PartitionParticles.Normalized Index In Partition
       )
      )
     )
    )) {isSetParameters=true}
   &ScaleRibbonWidth(Ribbon Width Scale=&Lerp_Float(
     A=1
     B=2
     Alpha=&FloatFromCurve(
      FloatCurve=DataInterface() {propertyValues {Curve {keys [
          {interpMode=RCIM_Cubic, tangentMode=RCTM_User, time=0, value=0}
          {arriveTangent=2.2695, interpMode=RCIM_Cubic, leaveTangent=2.2695, tangentMode=RCTM_User, time=1, value=1}
         ]}}}
      CurveIndex=$Particles.PartitionParticles.Normalized Index In Partition
     )
    ))
   &ScaleColor(
    Scale Alpha=&FloatFromCurve(
     FloatCurve=DataInterface() {propertyValues {Curve {keys [<at 0 1>,<at 1 0>]}}}
     CurveIndex=$Particles.PartitionParticles.Normalized Index In Partition
     Scale Curve=0.7
    )
    Scale RGB=&MakeVectorFromLinearColorRGB(Color=$System.TODColor)
    ScaleA=true
    ScaleRGB=true
   )
  ]
  rendererClasses [&NiagaraRibbonRendererProperties]
  renderers [{rendererClass=&NiagaraRibbonRendererProperties}]
 }
}
summary {
 systemName=NS_Shoe_GlitterMane_RainbowTrails
 emitters [
  {emitterName=NE_RainbowTrail, rendererClasses [&NiagaraRibbonRendererProperties]}
  {emitterName=NE_Glints, rendererClasses [&NiagaraSpriteRendererProperties]}
 ]
 userVariables [
  User.HideFootFX {type=NiagaraBool}
  User.ShoeScale_L {type=Vector3f, description="", defaultValue {struct=Vector3f, value=<1 1 1>}}
  User.ShoeScale_R {type=Vector3f, description="", defaultValue {struct=Vector3f, value=<1 1 1>}}
 ]
}
